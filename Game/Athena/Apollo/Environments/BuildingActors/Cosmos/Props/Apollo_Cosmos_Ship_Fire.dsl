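; Building actor prop for a ship fire in the Cosmos environment set, whose collision can burn players.
; A ShouldDie override always returns false, noted as clamping minimum health to 1, so damage does not destroy it.
; The construction script enables actor collision only when the actor is not hidden, so invisible fire cannot burn players, then calls the parent construction script.
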(fn ShouldDie
  (params (Damage :type Float%20%28single-precision%29) (EventInstigator :type Controller%20Object%20Reference) (DamageCauser :type Actor%20Object%20Reference) (bIsTest :type Boolean) (EffectContext :type Gameplay%20Effect%20Context%20Handle%20Structure%20%28by%20ref%29))
  ;[1 1 1] Clamps min hp to 1
  (return false))

(fn ConstructionScript
  ;[1 1 1] Ensure that if the visuals are off the player will not be burnt by fire collision
  (Collision|SetActorEnableCollision (not (Variables|Rendering|GetbHidden)))
  (|CallParentFunction))
