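; Smoke-and-fire crash site prop for the Apollo Cosmos environment, built as an indestructible building actor with a POI identifier.
; Its construction script disables collision while hidden so invisible fire cannot burn players, and ShouldDie always returns false.
; A custom event activates the near particle system after a zero-length delay, working around activation in the frame the actor is unhidden.
; The fire sound fades out when play ends.
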(event Custom|EnableNearParticles
  ;[0 0 0] fix for activating, one frame delay after being set unhidden in game
  (Utilities|FlowControl|Delay 0.0)
  (Components|Activation|Activate (Variables|Default|GetNear%20Particle%20system) true))

(event EventEndPlay
  (params (EndPlayReason :type EEndPlayReason%20Enum :default "Destroyed"))
  (Audio|Components|Audio|FadeOut (Variables|Default|GetFire_Sound) 0.5 0.0 "Linear" :!class /Script/Engine.AudioComponent))

(fn GetPOIIdentifierValue
  (return (Variables|Default|GetPOI_Identifier)))

(fn ShouldDie
  (params (Damage :type Float%20%28single-precision%29) (EventInstigator :type Controller%20Object%20Reference) (DamageCauser :type Actor%20Object%20Reference) (bIsTest :type Boolean) (EffectContext :type Gameplay%20Effect%20Context%20Handle%20Structure%20%28by%20ref%29))
  ;[1 1 1] Clamps min hp to 1
  (return false))

(fn ConstructionScript
  ;[1 1 1] Ensure that if the visuals are off the player will not be burnt by fire collision
  (Collision|SetActorEnableCollision (not (Variables|Rendering|GetbHidden)))
  (|CallParentFunction))
